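# Sound cue for Fiend Hunter dual crossbow projectile impacts on players.
# A Random node picks one of five equally weighted impact wave variations, which a Modulator then varies slightly in pitch and volume.
# It plays in the ranged weapon sound class with arrow impact attenuation, reduced volume and a concurrency override limiting simultaneous instances.
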SoundCue /CorruptionGameplay/Sounds/Weapons/FiendHunterDualies/FiendHunnterDualies_Proj_Impt_Player_Cue
root=Modulator_8
nodes [
 {id=Modulator_8, children [Random_1], properties {pitchMax=1.1, pitchMin=0.9, volumeMax=1.1, volumeMin=0.9}}
 {id=Random_1, children [WavePlayer_6, WavePlayer_7, WavePlayer_8, WavePlayer_9, WavePlayer_10], properties {weights [1,1,1,1,1]}}
 {
  id=WavePlayer_10
  properties {soundWaveAssetPtr=&/Crossbow_Dual_FiendHunter_Proj_Impt_Player_05}
 }
 {
  id=WavePlayer_6
  properties {soundWaveAssetPtr=&/Crossbow_Dual_FiendHunter_Proj_Impt_Player_01}
 }
 {
  id=WavePlayer_7
  properties {soundWaveAssetPtr=&/Crossbow_Dual_FiendHunter_Proj_Impt_Player_02}
 }
 {
  id=WavePlayer_8
  properties {soundWaveAssetPtr=&/Crossbow_Dual_FiendHunter_Proj_Impt_Player_03}
 }
 {
  id=WavePlayer_9
  properties {soundWaveAssetPtr=&/Crossbow_Dual_FiendHunter_Proj_Impt_Player_04}
 }
]
properties {
 attenuationSettings=/Game/Sounds/Attenuation/Fort_Arrow_Impact_Athena_Att
 maxDistance=6000
 soundClassObject=&Fort_WeaponRanged
 volumeMultiplier=0.7
 concurrencyOverrides {maxCount=3}
}
