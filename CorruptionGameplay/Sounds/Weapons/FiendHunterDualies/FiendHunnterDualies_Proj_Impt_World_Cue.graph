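# Sound cue for Fiend Hunter dual crossbow projectile impacts on world surfaces, classed as ranged weapon audio.
# A Random node picks among six equally weighted impact wave variations, feeding a Modulator that slightly varies pitch and volume on each play.
# It plays at reduced volume with an arrow impact attenuation asset and a concurrency override capping simultaneous instances.
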SoundCue /CorruptionGameplay/Sounds/Weapons/FiendHunterDualies/FiendHunnterDualies_Proj_Impt_World_Cue
root=Modulator_8
nodes [
 {id=Modulator_8, children [Random_0], properties {pitchMax=1.1, pitchMin=0.9, volumeMax=1.1, volumeMin=0.9}}
 {id=Random_0, children [WavePlayer_0,WavePlayer_1,WavePlayer_2,WavePlayer_3,WavePlayer_4,WavePlayer_5], properties {weights [1,1,1,1,1,1]}}
 {
  id=WavePlayer_0
  properties {soundWaveAssetPtr=&/Crossbow_Dual_FiendHunter_Proj_Impt_World_01}
 }
 {
  id=WavePlayer_1
  properties {soundWaveAssetPtr=&/Crossbow_Dual_FiendHunter_Proj_Impt_World_02}
 }
 {
  id=WavePlayer_2
  properties {soundWaveAssetPtr=&/Crossbow_Dual_FiendHunter_Proj_Impt_World_03}
 }
 {
  id=WavePlayer_3
  properties {soundWaveAssetPtr=&/Crossbow_Dual_FiendHunter_Proj_Impt_World_04}
 }
 {
  id=WavePlayer_4
  properties {soundWaveAssetPtr=&/Crossbow_Dual_FiendHunter_Proj_Impt_World_05}
 }
 {
  id=WavePlayer_5
  properties {soundWaveAssetPtr=&/Crossbow_Dual_FiendHunter_Proj_Impt_World_06}
 }
]
properties {
 attenuationSettings=/Game/Sounds/Attenuation/Fort_Arrow_Impact_Athena_Att
 maxDistance=6000
 soundClassObject=&Fort_WeaponRanged
 volumeMultiplier=0.4
 concurrencyOverrides {maxCount=3}
}
