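# Sound cue for Fiend Hunter dual-crossbow arrows hitting water.
# A random node picks one of four equally weighted water impact waves, then a modulator slightly varies its pitch and volume.
# It plays boosted in the impacts sound class with arrow-impact distance attenuation and a concurrency override that allows only a couple of simultaneous instances.
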SoundCue /CorruptionGameplay/Sounds/Weapons/FiendHunterDualies/FiendHunterDualies_Arrow_Impact_Water_Cue
root=Modulator_8
nodes [
 {id=Modulator_8, children [Random_0], properties {pitchMax=1.1, pitchMin=0.9, volumeMax=1.1, volumeMin=0.9}}
 {id=Random_0, children [WavePlayer_0,WavePlayer_1,WavePlayer_2,WavePlayer_3], properties {weights [1,1,1,1]}}
 {id=WavePlayer_0, properties {soundWaveAssetPtr=&/Crossbow_Arrow_Impact_Water_01}}
 {id=WavePlayer_1, properties {soundWaveAssetPtr=&/Crossbow_Arrow_Impact_Water_02}}
 {id=WavePlayer_2, properties {soundWaveAssetPtr=&/Crossbow_Arrow_Impact_Water_03}}
 {id=WavePlayer_3, properties {soundWaveAssetPtr=&/Crossbow_Arrow_Impact_Water_04}}
]
properties {
 attenuationSettings=/Game/Sounds/Attenuation/Fort_Arrow_Impact_Athena_Att
 bOverrideConcurrency=true
 maxDistance=6000
 soundClassObject=&Fort_Impacts
 volumeMultiplier=4.032
 concurrencyOverrides {maxCount=2}
}
