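# Sound cue for the first-person firing sound of the Fiend Hunter dual crossbows.
# A Random node picks among five equally weighted close-range shot recordings and feeds a Modulator at default settings for slight pitch and volume variation.
# The cue plays at reduced volume through the first-person ranged-fire sound class, using first-person weapon attenuation and the shared first-person gunfire concurrency group.
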SoundCue /CorruptionGameplay/Sounds/Weapons/FiendHunterDualies/FiendHunterDualies_Fire_1P_Cue
root=Modulator_0
nodes [
 {id=Modulator_0, children [Random_1]}
 {id=Random_1, children [WavePlayer_0,WavePlayer_6,WavePlayer_7,WavePlayer_8,WavePlayer_9], properties {weights [1,1,1,1,1]}}
 {
  id=WavePlayer_0
  properties {soundWaveAssetPtr=&/Crossbow_Dual_FiendHunter_Shoot_Close_1p_01}
 }
 {
  id=WavePlayer_6
  properties {soundWaveAssetPtr=&/Crossbow_Dual_FiendHunter_Shoot_Close_1p_03}
 }
 {
  id=WavePlayer_7
  properties {soundWaveAssetPtr=&/Crossbow_Dual_FiendHunter_Shoot_Close_1p_04}
 }
 {
  id=WavePlayer_8
  properties {soundWaveAssetPtr=&/Crossbow_Dual_FiendHunter_Shoot_Close_1p_05}
 }
 {
  id=WavePlayer_9
  properties {soundWaveAssetPtr=&/Crossbow_Dual_FiendHunter_Shoot_Close_1p_06}
 }
]
properties {
 attenuationSettings=&Fort_Weapon_Projectile_FirstPerson_Athena_Att
 maxDistance=2097152
 soundClassObject=&Fort_WeaponRanged_Fire_1P_PSM
 volumeMultiplier=0.7
 attenuationOverrides {falloffDistance=5000, attenuationShapeExtents {x=1000}}
 concurrencyOverrides {resolutionRule=PreventNew}
 concurrencySet [&Weapon_Gunfire_FirstPerson_Conc]
}
