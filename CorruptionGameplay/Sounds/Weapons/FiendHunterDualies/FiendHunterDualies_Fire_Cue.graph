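# Third-person fire sound cue for the FiendHunter dual crossbows, mixing a close layer with a medium-range layer.
# Each layer picks randomly among six equally weighted shot waves, then passes through its own Athena bow attenuation, close or distant.
# Sound class overrides send the layers to the 3P and distant ranged-fire classes before the mixer combines them at reduced volume.
# Global and local gunfire concurrency sets limit simultaneous shots.
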SoundCue /CorruptionGameplay/Sounds/Weapons/FiendHunterDualies/FiendHunterDualies_Fire_Cue
root=Mixer_1
nodes [
 {id=Attenuation_3, children [Random_1], properties {attenuationSettings=/Game/Sounds/Attenuation/Fort_Weapon_Projectile_Close_Bow_Athena_Att}}
 {
  id=Attenuation_4
  children [Random_3]
  properties {attenuationSettings=/Game/Sounds/Attenuation/Fort_Weapon_Projectile_Distant_Bow_Athena_Att}
 }
 {id=Mixer_1, children [SoundClass_3,SoundClass_2], properties {inputVolume [0.6,0.6]}}
 {
  id=Random_1
  children [WavePlayer_0, WavePlayer_1, WavePlayer_2, WavePlayer_7, WavePlayer_8, WavePlayer_13]
  properties {weights [1,1,1,1,1,1]}
 }
 {
  id=Random_3
  children [WavePlayer_14, WavePlayer_15, WavePlayer_16, WavePlayer_17, WavePlayer_18, WavePlayer_19]
  properties {weights [1,1,1,1,1,1]}
 }
 {
  id=SoundClass_2
  children [Attenuation_4]
  properties {soundClassOverride=&Fort_WeaponRanged_Fire_Distant}
 }
 {
  id=SoundClass_3
  children [Attenuation_3]
  properties {soundClassOverride=&Fort_WeaponRanged_Fire_3P_PSM}
 }
 {
  id=WavePlayer_0
  properties {soundWaveAssetPtr=&/Crossbow_Dual_FiendHunter_Shoot_Close_3p_01}
 }
 {
  id=WavePlayer_1
  properties {soundWaveAssetPtr=&/Crossbow_Dual_FiendHunter_Shoot_Close_3p_02}
 }
 {
  id=WavePlayer_13
  properties {soundWaveAssetPtr=&/Crossbow_Dual_FiendHunter_Shoot_Close_3p_06}
 }
 {
  id=WavePlayer_14
  properties {soundWaveAssetPtr=&/Crossbow_Dual_FiendHunter_Shoot_Med_01}
 }
 {
  id=WavePlayer_15
  properties {soundWaveAssetPtr=&/Crossbow_Dual_FiendHunter_Shoot_Med_02}
 }
 {
  id=WavePlayer_16
  properties {soundWaveAssetPtr=&/Crossbow_Dual_FiendHunter_Shoot_Med_03}
 }
 {
  id=WavePlayer_17
  properties {soundWaveAssetPtr=&/Crossbow_Dual_FiendHunter_Shoot_Med_04}
 }
 {
  id=WavePlayer_18
  properties {soundWaveAssetPtr=&/Crossbow_Dual_FiendHunter_Shoot_Med_05}
 }
 {
  id=WavePlayer_19
  properties {soundWaveAssetPtr=&/Crossbow_Dual_FiendHunter_Shoot_Med_06}
 }
 {
  id=WavePlayer_2
  properties {soundWaveAssetPtr=&/Crossbow_Dual_FiendHunter_Shoot_Close_3p_03}
 }
 {
  id=WavePlayer_7
  properties {soundWaveAssetPtr=&/Crossbow_Dual_FiendHunter_Shoot_Close_3p_04}
 }
 {
  id=WavePlayer_8
  properties {soundWaveAssetPtr=&/Crossbow_Dual_FiendHunter_Shoot_Close_3p_05}
 }
]
properties {
 maxDistance=27500
 volumeMultiplier=0.7
 attenuationOverrides {falloffDistance=5000, attenuationShapeExtents {x=1000}}
 concurrencyOverrides {resolutionRule=PreventNew}
 concurrencySet [&Weapon_Gunfire_Global_Conc, &Weapon_Gunfire_Local_Conc]
}
