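# Sound cue for the pass-by sound of Fiend Hunter dual crossbow projectiles.
# A quality-level node sends its first two levels to a random pick of six waves and its last two to a pick of three, each through a slight upward pitch modulator.
# It plays quietly in the ranged-weapon sound class with compound bow whoosh attenuation, overrides concurrency to cap simultaneous plays, and has the debug flag on.
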SoundCue /CorruptionGameplay/Sounds/Weapons/FiendHunterDualies/FiendHunterDualies_Proj_Pass_By_Cue
root=QualityLevel_0
nodes [
 {id=Modulator_0, children [Random_0], properties {pitchMax=1.15, pitchMin=1}}
 {id=Modulator_1, children [Random_1], properties {pitchMax=1.15, pitchMin=1}}
 {id=QualityLevel_0, children [Modulator_0,Modulator_0,Modulator_1,Modulator_1]}
 {id=Random_0, children [WavePlayer_1,WavePlayer_2,WavePlayer_3,WavePlayer_4,WavePlayer_8,WavePlayer_9], properties {weights [1,1,1,1,1,1]}}
 {id=Random_1, children [WavePlayer_0,WavePlayer_5,WavePlayer_6], properties {weights [1,1,1]}}
 {
  id=WavePlayer_0
  properties {soundWaveAssetPtr=&/Crossbow_Dual_FiendHunter_Proj_Pass_By_01}
 }
 {
  id=WavePlayer_1
  properties {soundWaveAssetPtr=&Crossbow_Dual_FiendHunter_Proj_Pass_By_01}
 }
 {
  id=WavePlayer_2
  properties {soundWaveAssetPtr=&/Crossbow_Dual_FiendHunter_Proj_Pass_By_02}
 }
 {
  id=WavePlayer_3
  properties {soundWaveAssetPtr=&/Crossbow_Dual_FiendHunter_Proj_Pass_By_03}
 }
 {
  id=WavePlayer_4
  properties {soundWaveAssetPtr=&/Crossbow_Dual_FiendHunter_Proj_Pass_By_04}
 }
 {
  id=WavePlayer_5
  properties {soundWaveAssetPtr=&Crossbow_Dual_FiendHunter_Proj_Pass_By_02}
 }
 {
  id=WavePlayer_6
  properties {soundWaveAssetPtr=&Crossbow_Dual_FiendHunter_Proj_Pass_By_03}
 }
 {
  id=WavePlayer_8
  properties {soundWaveAssetPtr=&/Crossbow_Dual_FiendHunter_Proj_Pass_By_05}
 }
 {
  id=WavePlayer_9
  properties {soundWaveAssetPtr=&/Crossbow_Dual_FiendHunter_Proj_Pass_By_06}
 }
]
properties {
 attenuationSettings=/Game/Sounds/Fort_Weapons/Explosive_Bow/Weapon_Compound_Bow_Whoosh_By_Att
 bDebug=true
 bOverrideConcurrency=true
 maxDistance=2097152
 soundClassObject=&Fort_WeaponRanged
 volumeMultiplier=0.3
 attenuationOverrides {
  bAttenuate=false
  bEnableReverbSend=false
  bUseComplexCollisionForOcclusion=true
  falloffDistance=0
  occlusionVolumeAttenuation=0
  stereoSpread=0
  attenuationShapeExtents {x=50000}
 }
 concurrencyOverrides {maxCount=3}
}
