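# Sound cue for the clip-insert step of reloading the Fiend Hunter dual crossbows, in the weapon foley sound class.
# A single wave player feeds the output directly, playing one clip-insert recording with no randomization or modulation.
# It uses the shared weapon reload attenuation and a slightly lowered volume, overrides concurrency, and has its debug flag enabled.
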SoundCue /CorruptionGameplay/Sounds/Weapons/FiendHunterDualies/FiendHunterDualies_Reload_Insert_Cue
root=WavePlayer_3
nodes [{
  id=WavePlayer_3
  properties {soundWaveAssetPtr=&/Crossbow_Dual_FiendHunter_Reload_ClipInsert_01}
 }]
properties {
 attenuationSettings=&Fort_Weapon_Reload_Att
 bDebug=true
 bOverrideConcurrency=true
 maxDistance=2250
 soundClassObject=&Fort_WeaponFoley
 volumeMultiplier=0.752
}
