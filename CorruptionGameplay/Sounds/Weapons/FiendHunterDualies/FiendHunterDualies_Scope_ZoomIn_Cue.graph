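# Sound cue for the scope zoom-in (aim-down-sights) sound of the Fiend Hunter dual crossbows.
# A single wave player feeds a modulator with default settings, played under the weapon foley sound class at reduced volume.
# Overridden attenuation sets custom falloff and a non-spatialized near radius, and overridden concurrency rejects new instances when the limit is reached instead of stopping playing ones.
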SoundCue /CorruptionGameplay/Sounds/Weapons/FiendHunterDualies/FiendHunterDualies_Scope_ZoomIn_Cue
root=Modulator_2
nodes [
 {id=Modulator_2, children [WavePlayer_3]}
 {id=WavePlayer_3, properties {soundWaveAssetPtr=&/Crossbow_Dual_FiendHunter_ADS_01}}
]
properties {
 bOverrideAttenuation=true
 bOverrideConcurrency=true
 maxDistance=2700
 soundClassObject=&Fort_WeaponFoley
 volumeMultiplier=0.251
 attenuationOverrides {falloffDistance=1500, nonSpatializedRadiusStart=200, attenuationShapeExtents {x=1200}}
 concurrencyOverrides {resolutionRule=PreventNew}
}
